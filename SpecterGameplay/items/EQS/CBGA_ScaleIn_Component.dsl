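; Actor component for CBGA_Parent consumable building actors that grows the owner's static mesh in on spawn and shrinks it out before removal.
; A looping timer eases a uniform scale via FInterpTo through SetOwnerMeshScale, snapping and clearing itself on arrival.
; A lifetime timer calls ScaleOutAndDie to limit world accumulation.
; Scale-out waits a random moment, reuses the lerp back to start scale, then silently kills the owner after a short delay.
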(event EventBeginPlay
  (bind _ascbga_parent (Utilities|Casting|CastToCBGA_Parent (Components|GetOwner)))
  (Variables|Default|SetOwner%20As%20CBGA%20Parent _ascbga_parent)
  (bind _output_get (Variables|Default|SetCur%20Min%20World%20Scale (Variables|Default|GetStart%20World%20Scale)))
  (Default|SetOwnerMeshScale _output_get)
  (Variables|Default|SetCur%20Max%20World%20Scale (Variables|Default|GetDest%20World%20Scale))
  (Variables|Default|SetHasReachedDestScale false)
  (bind _returnvalue (Utilities|Time|SetTimerbyEvent (delegate Custom|UpdateLerpToDest) (Variables|Default|GetLerpTimerInterval) true))
  (Variables|Default|SetLerpTimerHandle _returnvalue)
  ; To prevent a zillion of these accumulating in the world, kill these after a specified amount of time.
  (Utilities|Time|SetTimerbyEvent (delegate Custom|LifeTimeExpired) 1.2e+02))

(event Custom|Scale%20Out%20and%20Die
  (Variables|Default|SetScaleOutAndDie true)
  (Utilities|FlowControl|Delay (Math|Random|RandomFloatinRange 0.0 0.4))
  (Variables|Default|SetCur%20Min%20World%20Scale (Variables|Default|GetCurr%20World%20Scale))
  (Variables|Default|SetCur%20Max%20World%20Scale (Variables|Default|GetStart%20World%20Scale))
  (Variables|Default|SetHasReachedDestScale false)
  (bind _returnvalue (Utilities|Time|SetTimerbyEvent (delegate Custom|UpdateLerpToDest) (Variables|Default|GetLerpTimerInterval) true))
  (Variables|Default|SetLerpTimerHandle _returnvalue)
  ; To prevent a zillion of these accumulating in the world, kill these after a specified amount of time.
  (Utilities|Time|SetTimerbyEvent (delegate Custom|LifeTimeExpired) 1.2e+02))

(event Custom|LifeTimeExpired
  ; To prevent a zillion of these accumulating in the world, kill these after a specified amount of time.
  (CallFunction|ScaleOutAndDie))

(event Custom|UpdateLerpToDest
  ;[1 1 1] Lerp Scale as needed
  (Utilities|FlowControl|Sequence
    (:then_0
      (if (not (Variables|Default|GetHasReachedDestScale))
        (Default|SetOwnerMeshScale (Math|Interpolation|FInterpTo (Variables|Default|GetCurr%20World%20Scale) (Variables|Default|GetCur%20Max%20World%20Scale) (Variables|Default|GetLerpTimerInterval) (Variables|Default|GetScale%20Interp%20Speed)))
        (if (<= (Math|Float|Absolute(Float) (- (Variables|Default|GetCurr%20World%20Scale) (Variables|Default|GetCur%20Max%20World%20Scale))) 0.02)
          (Default|SetOwnerMeshScale (Variables|Default|GetCur%20Max%20World%20Scale))
          (Variables|Default|SetHasReachedDestScale true))))
    (:then_1
      ;[1 1 1] Cancel looping once we've hit our target
      (if (Variables|Default|GetHasReachedDestScale)
        (Utilities|Time|ClearandInvalidateTimerbyHandle (Variables|Default|GetLerpTimerHandle))
        (if (Variables|Default|GetScaleOutAndDie)
          (Utilities|FlowControl|Delay 1.0)
          (bind _ascbga_parent (Utilities|Casting|CastToCBGA_Parent (Components|GetOwner)))
          (Gameplay|SilentDie _ascbga_parent false :!class /Script/FortniteGame.BuildingActor))))))

(fn SetOwnerMeshScale
  (params (SizeAlpha :type Float%20%28double-precision%29))
  (bind _output_get (Variables|Default|SetCurr%20World%20Scale SizeAlpha))
  (bind (_returnvalue_x _returnvalue_y _returnvalue_z) (Math|Vector|BreakVector (* (Variables|Default|GetDefaultScaleMult) _output_get)))
  (Transformation|SetWorldScale3D (Variables|/Game/Athena/BuildingActors/ConsumableBGAs/CBGA_Parent.CBGA_Parent_C|GetStaticMesh (Variables|Default|GetOwner%20As%20CBGA%20Parent)) (Math|Vector|MakeVector _returnvalue_x _returnvalue_y _returnvalue_z)))
